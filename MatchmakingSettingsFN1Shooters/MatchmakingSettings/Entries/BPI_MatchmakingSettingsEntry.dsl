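; Blueprint interface for matchmaking settings entries, declaring a GetAnalyticsAttribute function.
; The function body holds only a default return, so implementing entries supply their own analytics attribute.
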(fn GetAnalyticsAttribute
  (return 0))
