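# Icon toggle button for FN1 matchmaking settings, a child of the settings entry base that serves as a team-size list entry.
# Binding a list item stores its team-size view model, resets transitions and flags the next selection as instant.
# Selection changes drive UIKit block transitions and selected state.
# Helper functions set enabled, locked and selectable state, load the icon from a CMS URL, and toggle the background material's IsPureChaos parameter.
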WidgetBlueprint /MatchmakingSettingsFN1Shooters/MatchmakingSettings/Entries/Buttons/WBP_FN1_ToggleIconButton
graphs [
 EventGraph {nodes [
   n1 = Event PreConstruct()
   > Parent: PreConstruct(IsDesignTime=@n1.IsDesignTime)
   > SetText(self:WBP FN1 Matchmaking Settings Entry Base Object Reference, InText:Text)
   > SetMaterial!(
    execute
    self:WBP Matchmaking Block Image Object Reference=Get Block Icon() -> (Block Icon:WBP Matchmaking Block Image Object Reference)
    InMaterial:"Material Instance Object Reference (by ref)"=Get ImageMaterial() -> (ImageMaterial:Material Instance Object Reference)
   ) -> (then)
   Event On Selected()
   > Branch(Condition=Get bSelectable() -> (bSelectable:bool))
   > n2 = Branch(Condition=Get Is Instant Selection() -> (Is Instant Selection:bool))
   > RequestInstantTransitionOnBlocks!(execute, self:UIKit Modular Button Object Reference, State:EUIKitBlockInstantTransitionState Enum=InstantSelected) -> (then)
   > n3 = Set Is Instant Selection(Is Instant Selection:bool=false) -> (Output_Get:bool)
   n4 = For Each Loop(
    Exec=[@n6.then, @n3.then]
    Array:Array of UIKit Block Object References=Get ModularBlocks() -> (ModularBlocks:Array of UIKit Block Object References)
   ) -> (Array Element:UIKit Block Object Reference)
   > n5 = Cast To WBP_UIKit_Block_Base!(execute, Object=@n4) -> (then, CastFailed, AsWBP UIKit Block Base:WBP UIKit Block Base Object Reference) {is=title K2Node_DynamicCast}
   > SetIsSelected!(execute, self:WBP UIKit Block Base Object Reference=@n5, IsSelected:bool=true) -> (then)
   n6 = RequestTransitionOnBlocks!(execute=@n2.else, self:UIKit Modular Button Object Reference, State:EUIKitBlockVisualState Enum=Selected) -> (then)
   Event On Deselected()
   > n7 = For Each Loop(Array=Get ModularBlocks())
   > n8 = Cast To WBP_UIKit_Block_Base(Object=@n7)
   > SetIsSelected(self=@n8, IsSelected=false)
   > Parent: On Deselected() {is=title K2Node_CallParentFunction}
   Event On Pressed() {is=title K2Node_Event}
   > Parent: On Pressed() {is=title K2Node_CallParentFunction}
   > Set Is Instant Selection(Is Instant Selection=false)
   n9 = Event OnListItemObjectSet()
   > n10 = Cast To VM_TeamSizeEnumEntry!(execute, Object=@n9.ListItemObject) -> (then, CastFailed, AsVM Team Size Enum Entry:VM Team Size Enum Entry Object Reference) {is=title K2Node_DynamicCast}
   > Set VM_ShowcaseEnumEntry(VM_ShowcaseEnumEntry:VM Team Size Enum Entry Object Reference=@n10) -> (Output_Get:VM Team Size Enum Entry Object Reference)
   > Stop All Transitions!(In:Exec) -> (Out:Exec) {is=title K2Node_MacroInstance}
   > Set Is Instant Selection(Is Instant Selection=true)
   > ResetAllStates(self:WBP Matchmaking Settings Background Object Reference=Get BackgroundBlock() -> (BackgroundBlock:WBP Matchmaking Settings Background Object Reference))
  ]}
 SetData {nodes [
   n1 = SetData!() -> (then, Data_bIsEnabled:bool=~, Data_Value:String, Data_Name:Text) {is=title K2Node_FunctionEntry}
   > SetIsLocked!(execute, self:Common Button Base Object Reference, bInIsLocked:bool=NOT Boolean(A=@n1)) -> (then)
   > SetEnabled!(execute, self:WBP FN1 Matchmaking Settings Entry Base Object Reference, Enabled:bool=@n1) -> (then)
   > SetIsSelectable!(execute, self:Common Button Base Object Reference, bInIsSelectable:bool=@n1) -> (then)
  ]}
 SetImageURL {nodes [
   n1 = SetImageURL!() -> (then, URL:String) {is=title K2Node_FunctionEntry}
   n2 = Get Block Icon()
   > SetVisibility(self=@n2, InVisibility=Select(Option 0=SelfHitTestInvisible, Option 1=Collapsed, Index=IsEmpty(InString=@n1)))
   > n3 = Branch(Condition=IsEmpty(InString=@n1))
   SetMediaURL!(
    execute=@n3.else
    self:Epic CMSImage Object Reference=Get EpicCMSImage!(self:WBP Matchmaking Block Image Object Reference=@n2) -> (EpicCMSImage:Epic CMSImage Object Reference)
    MediaUrl:String=@n1
   ) -> (then)
  ]}
 SetIsPureChaos {nodes [
   n1 = SetIsPureChaos!() -> (then, IsPureChaos:bool=~) {is=title K2Node_FunctionEntry}
   > SetScalarParameterValue(
    self=Get BackgroundDynamicMaterial!(self:WBP UIKit Block Background Base Object Reference=Get BackgroundBlock()) -> (BackgroundDynamicMaterial:Material Instance Dynamic Object Reference)
    ParameterName=IsPureChaos
    Value="To Float (Boolean)"(InBool=@n1)
   )
  ]}
]
variables [
 IconsMap {
  type="map<Texture2D>"
  default {
   duo=/FortUILibrary/Textures/Icons/T_UI_Icon_UIKit_CurrentPlayerCount_MSDF
   sixStack=/FortUILibrary/Textures/Icons/T_UI_Icon_UIKit_MorePlayers_MSDF
   solo=/FortUILibrary/Textures/Icons/T_UI_Icon_UIKit_Character_MSDF
   squad=/FortUILibrary/Textures/Icons/T_UI_Icon_UiKit_Squads_MSDF
   trio=/FortUILibrary/Textures/Icons/T_UI_Icon_UIKit_PartyUp_MSDF
  }
 }
 ImageMaterial {type=MaterialInstance, default=/MatchmakingSettingsFN1Shooters/MatchmakingSettings/MaterialInstances/MI_UI_FN1_Icon}
]
widgets {
 info {
  inheritedWidgetCount=0
  parentClass=&/WBP_FN1_MatchmakingSettings_EntryBase.WBP_FN1_MatchmakingSettings_EntryBase_C
  rootWidgetClass=null
  widgetCount=0
 }
 widgets []
}
