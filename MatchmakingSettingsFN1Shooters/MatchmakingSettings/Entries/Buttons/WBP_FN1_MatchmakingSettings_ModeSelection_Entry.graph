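# Mode selection button entry for the matchmaking settings UI, a list-item child of WBP_FN1_MatchmakingSettings_EntryBase that adds no widgets of its own.
# When a list item is set, it casts the object to UIKitButtonViewModel and stores it as the entry's view model.
# Set DisabledState drives an IsDisabled scalar on the background dynamic material and dims render opacity.
# PreConstruct resets it to enabled; this graph never creates that material instance.
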WidgetBlueprint /MatchmakingSettingsFN1Shooters/MatchmakingSettings/Entries/Buttons/WBP_FN1_MatchmakingSettings_ModeSelection_Entry
graphs [
 EventGraph {nodes [
   n1 = Event PreConstruct()
   > Parent: PreConstruct(IsDesignTime=@n1.IsDesignTime)
   > Set DisabledState!(execute, self, IsDisabled:bool=false) -> (then) {is=title K2Node_CallFunction}
   n2 = Event OnListItemObjectSet()
   > n3 = Cast To UIKitButtonViewModel!(execute, Object=@n2.ListItemObject) -> (then, CastFailed, AsUI Kit Button View Model:UI Kit Button View Model Object Reference) {is=title K2Node_DynamicCast}
   > Set UIKitButtonViewModel(UIKitButtonViewModel:UI Kit Button View Model Object Reference=@n3) -> (Output_Get:UI Kit Button View Model Object Reference)
  ]}
 Set DisabledState {nodes [
   n1 = Set DisabledState!() -> (then, IsDisabled:bool=~) {is=title K2Node_FunctionEntry}
   > SetScalarParameterValue(
    self=Get BackgroundDynamicMaterial() -> (BackgroundDynamicMaterial:Material Instance Dynamic Object Reference)
    ParameterName=IsDisabled
    Value=Select(Option 0:float=0, Option 1:float=1, Index=@n1) -> (ReturnValue:float)
   )
   > Set RenderOpacity!(execute, RenderOpacity:float=Select(Option 0=1, Option 1=0.6, Index=@n1), self:Widget Object Reference) -> (then, Output_Get:float)
  ]}
]
variables [BackgroundDynamicMaterial {type=MaterialInstanceDynamic}]
widgets {
 info {
  inheritedWidgetCount=0
  parentClass=&/WBP_FN1_MatchmakingSettings_EntryBase.WBP_FN1_MatchmakingSettings_EntryBase_C
  rootWidgetClass=null
  widgetCount=0
 }
 widgets []
}
