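# Matchmaking settings list entry widget that lets the player choose a mode from two side-by-side buttons backed by a discrete-enum setting view model.
# A button group requires a selection, writes the chosen value to the view model and fires an analytics interaction event.
# SetButtonsData labels and enables the buttons from options, and SetSelectedMode syncs selection.
# Hover drives the description, touch focus also shows it, and focus shows an outline and forwards to the selected button.
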WidgetBlueprint /MatchmakingSettingsFN1Shooters/MatchmakingSettings/Entries/WBP_FN1_MatchmakingSettings_ModeSelection
graphs [
 EventGraph {nodes [
   n1 = "On Hovered (EntryL)"!() -> (OutputDelegate, then, Button:Common Button Base Object Reference) {is=title K2Node_ComponentBoundEvent}
   n2 = Get Description() -> (Description:WBP FN1 Matchmaking Entry Description Object Reference)
   SetVisibility(execute=[@n1.then, @n4.then], self=@n2, InVisibility=SelfHitTestInvisible)
   n3 = "On Unhovered (EntryL)"!() -> (OutputDelegate, then, Button:Common Button Base Object Reference) {is=title K2Node_ComponentBoundEvent}
   SetVisibility(execute=[@n3.then, @n5.then], self=@n2, InVisibility=Hidden)
   n4 = "On Hovered (EntryR)"!() -> (OutputDelegate, then, Button:Common Button Base Object Reference) {is=title K2Node_ComponentBoundEvent}
   n5 = "On Unhovered (EntryR)"!() -> (OutputDelegate, then, Button:Common Button Base Object Reference) {is=title K2Node_ComponentBoundEvent}
   n6 = OnButtonSelected!() -> (OutputDelegate, then, AssociatedButton:Common Button Base Object Reference, ButtonIndex:int) {is=title K2Node_CustomEvent}
   n7 = Get FortMatchmakingSettingDiscreteEnumVM() -> (FortMatchmakingSettingDiscreteEnumVM:Fort Matchmaking Setting Discrete Enum VM Object Reference)
   > Is Valid(InputObject=@n7)
   n8 = "Get (a ref)"!(
    Array:Array of Matchmaking Setting Option Structures=GetOptions!(self:Fort Matchmaking Setting Discrete Enum VM Object Reference=@n7) -> (ReturnValue:Array of Matchmaking Setting Option Structures)
    Dimension 1=@n6.ButtonIndex
   ) -> (Output_bIsEnabled:bool=~, Output_Value:String, Output_Name:Text)
   > SetValue!(
    execute
    self:Fort Matchmaking Setting Discrete Enum VM Object Reference=@n7
    Value:String=@n8.Output_Value
    LocalPlayer:Local Player Object Reference=GetOwningLocalPlayer!(self:Widget Object Reference) -> (ReturnValue:Local Player Object Reference)
   ) -> (then)
   > n9 = Set!(
    execute
    AnalyticsAttribute_Name:String=GetKey!(self:Fort Matchmaking Setting Discrete VM Object Reference=@n7) -> (ReturnValue:String)
    AnalyticsAttribute_Value:String="To String (Name)"(InName=String To Name(InString=@n8.Output_Value))
   ) -> (then, Output_Get_Name:String, Output_Get_Value:String)
   > Fire Event Element Interacted(
    Element=Self-Reference()
    AdditionalInfo="Add String (Non-Localized)"!(
     Info:"Flow Analytics Additional Info Structure (by ref)"=Make Flow Analytics Additional Info!() -> (FlowAnalyticsAdditionalInfo:Flow Analytics Additional Info Structure) {is=title K2Node_MakeStruct}
     Key:String=@n9
     Value:String=@n9.Output_Get_Value
    ) -> (ReturnValue:Flow Analytics Additional Info Structure)
   )
   n10 = OnButtonHovered!() -> (OutputDelegate, then, AssociatedButton:Common Button Base Object Reference, ButtonIndex:int) {is=title K2Node_CustomEvent}
   > UpdateDescriptionText!(execute, self, InButtonIndex:int=@n10.ButtonIndex) -> (then)
   Event PreConstruct()
   > n11 = Construct Object from Class!(execute, Class=/Script/CommonUI.CommonButtonGroupBase, self, bSelectionRequired:bool=true) -> (then, ReturnValue:Common Button Group Base Object Reference)
   > n12 = Set ButtonGroup(ButtonGroup:Common Button Group Base Object Reference=@n11) -> (Output_Get:Common Button Group Base Object Reference)
   > AddWidgets!(
    execute
    self:Common Widget Group Base Object Reference=@n12
    Widgets:Array of Widget Object References=Make Array!(
     "[0]":Widget Object Reference=Get EntryL() -> (EntryL:WBP FN1 Matchmaking Settings Mode Selection Entry Object Reference)
     "[1]":Widget Object Reference=Get EntryR() -> (EntryR:WBP FN1 Matchmaking Settings Mode Selection Entry Object Reference)
    ) -> (Array:Array of Widget Object References)
   ) -> (then)
   n13 = Create Event()
   > Bind Event to On Selected Button Base Changed(self:Common Button Group Base Object Reference=@n12, Delegate=@n13) {is=title K2Node_AddDelegate}
   n14 = Create Event()
   > Bind Event to On Hovered Button Base Changed(self:Common Button Group Base Object Reference=@n12, Delegate=@n14) {is=title K2Node_AddDelegate}
   Event OnAddedToFocusPath!() -> (OutputDelegate, then, InFocusEvent:Focus Event Structure) {is=title K2Node_Event}
   > ShowOutline!(
    execute
    self:WBP Matchmaking Settings Outline Object Reference=Get FocusOutline() -> (FocusOutline:WBP Matchmaking Settings Outline Object Reference)
    AdditionalCondition:bool=true
   ) -> (then)
   n15 = IsTouch!() -> (ReturnValue:bool)
   > Branch(Condition=@n15)
   n16 = Get Description()
   > SetVisibility(self=@n16, InVisibility=SelfHitTestInvisible)
   Event Destruct()
   > Unbind Event from On Selected Button Base Changed(self:Common Button Group Base Object Reference=@n12, Delegate=@n13) {is=title K2Node_RemoveDelegate}
   > Unbind Event from On Hovered Button Base Changed(self:Common Button Group Base Object Reference=@n12, Delegate=@n14) {is=title K2Node_RemoveDelegate}
   n17 = Event OnListItemObjectSet()
   > n18 = Cast To FortMatchmakingSettingDiscreteEnumVM!(execute, Object=@n17.ListItemObject) -> (then, CastFailed, AsFort Matchmaking Setting Discrete Enum VM:Fort Matchmaking Setting Discrete Enum VM Object Reference) {is=title K2Node_DynamicCast}
   > Set FortMatchmakingSettingDiscreteEnumVM(FortMatchmakingSettingDiscreteEnumVM:Fort Matchmaking Setting Discrete Enum VM Object Reference=@n18) -> (Output_Get:Fort Matchmaking Setting Discrete Enum VM Object Reference)
   Event OnRemovedFromFocusPath!() -> (OutputDelegate, then, InFocusEvent:Focus Event Structure) {is=title K2Node_Event}
   > HideOutline(self:WBP Matchmaking Settings Outline Object Reference=Get FocusOutline())
   > Branch(Condition=@n15)
   > SetVisibility(self=@n16, InVisibility=Hidden)
  ]}
 GetAnalyticsAttribute {nodes [
   GetAnalyticsAttribute() {is=title K2Node_FunctionEntry}
   > Return Node!(execute, Attribute:Analytics Event Attr Structure=Get AnalyticsAttribute() -> (AnalyticsAttribute:Analytics Event Attr Structure))
  ]}
 IsTouch {nodes [
   IsTouch() {is=title K2Node_FunctionEntry}
   n1 = Get CommonInputSubsystem()
   > Return Node!(execute, ReturnValue:bool="Equal (Enum)"(A=GetCurrentInputType(self=@n1), B=Touch))
  ]}
 OnFocusReceived {nodes [
   OnFocusReceived!() -> (then, MyGeometry:Geometry Structure, InFocusEvent:Focus Event Structure) {is=title K2Node_FunctionEntry}
   n1 = Get ButtonGroup() -> (ButtonGroup:Common Button Group Base Object Reference)
   > Return Node(ReturnValue:Event Reply Structure=SetUserFocus!(
     Reply:"Event Reply Structure (by ref)"=Handled()
     FocusWidget:Widget Object Reference=GetSelectedButtonBase!(self:Common Button Group Base Object Reference=@n1) -> (ReturnValue:Common Button Base Object Reference)
     bInAllUsers:bool=false
    ) -> (ReturnValue:Event Reply Structure))
  ]}
 SetButtonsData {nodes [
   n1 = SetButtonsData!() -> (then, Options:Array of Matchmaking Setting Option Structures) {is=title K2Node_FunctionEntry}
   > n2 = For Each Loop!(Exec, Array:Array of Matchmaking Setting Option Structures=@n1) -> (LoopBody, Array Element_bIsEnabled:bool=~, Array Element_Value:String, Array Element_Name:Text, Array Index, Completed)
   n3 = Get FortMatchmakingSettingDiscreteEnumVM()
   Set!(
    execute=@n2.Completed
    AnalyticsAttribute_Name=GetKey(self=@n3)
    AnalyticsAttribute_Value=GetValue!(self:Fort Matchmaking Setting Discrete Enum VM Object Reference=@n3) -> (ReturnValue:String)
   ) -> (then, Output_Get:Analytics Event Attr Structure)
   n4 = Cast To WBP_FN1_MatchmakingSettings_ModeSelection_Entry!(
    execute=@n2.LoopBody
    Object=GetButtonBaseAtIndex!(self:Common Button Group Base Object Reference=Get ButtonGroup(), Index:int=@n2.Array Index) -> (ReturnValue:Common Button Base Object Reference)
   ) -> (then, CastFailed, AsWBP FN1 Matchmaking Settings Mode Selection Entry:WBP FN1 Matchmaking Settings Mode Selection Entry Object Reference) {is=title K2Node_DynamicCast}
   > n5 = Construct Object from Class!(execute, Class=/Script/FortniteUI.UIKitButtonViewModel, self) -> (then, ReturnValue:UI Kit Button View Model Object Reference)
   > Set Label!(execute, Label:Text=@n2.Array Element_Name, self:UI Kit Button View Model Object Reference=@n5) -> (then, Output_Get:Text)
   > Set UIKitButtonViewModel!(
    execute
    UIKitButtonViewModel:UI Kit Button View Model Object Reference=@n5
    self:WBP FN1 Matchmaking Settings Mode Selection Entry Object Reference=@n4
   ) -> (then, Output_Get:UI Kit Button View Model Object Reference)
   > SetIsEnabled(self=@n4, bInIsEnabled=@n2)
  ]}
 SetSelectedMode {nodes [
   n1 = SetSelectedMode!() -> (then, ModeName:String) {is=title K2Node_FunctionEntry}
   > n2 = For Each Loop with Break!(
    Exec:Exec
    Array:Array of Matchmaking Setting Option Structures=GetOptions(self=Get FortMatchmakingSettingDiscreteEnumVM())
    Break:Exec=@n3.then
   ) -> (LoopBody:Exec, Array Element:Matchmaking Setting Option Structure, Array Index:int, Completed:Exec)
   Branch(execute=@n2.Completed, Condition=IsTouch() {is=title K2Node_CallFunction})
   > UpdateDescriptionText(InButtonIndex=GetSelectedButtonIndex!(self:Common Button Group Base Object Reference=Get ButtonGroup()) -> (ReturnValue:int=0))
   Branch(
    execute=@n2.LoopBody
    Condition="Equal Exactly (String)"(
     A=@n1
     B=Break Matchmaking Setting Option!(MatchmakingSettingOption:"Matchmaking Setting Option Structure (by ref)"=@n2) -> (bIsEnabled:bool=~, Value:String, Name:Text) {is=title K2Node_BreakStruct}.Value
    )
   )
   > n3 = SelectButtonAtIndex!(execute, self:Common Button Group Base Object Reference=Get ButtonGroup(), ButtonIndex:int=@n2.Array Index, bAllowSound:bool=false) -> (then)
  ]}
 UpdateDescriptionText {nodes [
   n1 = UpdateDescriptionText!() -> (then, InButtonIndex:int) {is=title K2Node_FunctionEntry}
   > Branch(Condition="integer > integer"(A=@n1, B=-1))
   > n2 = Get!(execute:Exec) -> (then:Exec, else:Exec, FortMatchmakingSettingDiscreteEnumVM:Fort Matchmaking Setting Discrete Enum VM Object Reference)
   n3 = GetOptions(self=@n2)
   > SetText(
    self:WBP FN1 Matchmaking Entry Description Object Reference=Get Description()
    InText:Text=GetOptionDescription!(
     self:Fort Matchmaking Setting Discrete Enum VM Object Reference=@n2
     Value:String="Get (a ref)"(Array=@n3, Dimension 1=@n1) -> (Output_bIsEnabled=~).Output_Value
    ) -> (ReturnValue:Text)
   )
  ]}
]
variables [
 AnalyticsAttribute {type=AnalyticsEventAttr, default {name="", value=""}}
 ButtonGroup {type=CommonButtonGroupBase, default=null}
]
widgets {
 info {inheritedWidgetCount=0, parentClass=&UserWidget, rootWidgetClass=&GridPanel, widgetCount=7}
 widgets [
  Root {widgetClassPath=&GridPanel, bIsVariable=true, widget=Root, properties {columnFill [1]}}
  HorizontalBox_0 {widgetClassPath=&HorizontalBox, parent=Root, widget=HorizontalBox_0, slot {class=GridSlot}}
  EntryL {
   widgetClassPath=/MatchmakingSettingsFN1Shooters/MatchmakingSettings/Entries/Buttons/WBP_FN1_MatchmakingSettings_ModeSelection_Entry
   bIsVariable=true
   parent=HorizontalBox_0
   widget=EntryL
   properties {minimumSize=<100 64>, minimumSize_Mobile=<100 80>, textPaddings=<ltrb 4 4 4 4>}
   slot {class=HorizontalBoxSlot, properties {size {sizeRule=Fill, value=1}}}
  }
  EntryR {
   widgetClassPath=&WBP_FN1_MatchmakingSettings_ModeSelection_Entry
   bIsVariable=true
   parent=HorizontalBox_0
   widget=EntryR
   properties {minimumSize=<100 64>, minimumSize_Mobile=<100 80>, textPaddings=<ltrb 4 4 4 4>}
   slot {class=HorizontalBoxSlot, properties {padding=<ltrb 16 0 0 0>, size {sizeRule=Fill, value=1}}}
  }
  FocusOutline {
   widgetClassPath=/MatchmakingSettingsFN1Shooters/MatchmakingSettings/WBP_Matchmaking_Settings_Outline
   bIsVariable=true
   parent=Root
   widget=FocusOutline
   slot {class=GridSlot, properties {layer=1, padding=<ltrb -28 -24 -28 -24>}}
  }
  Background {
   widgetClassPath=&Image
   bIsVariable=true
   parent=Root
   widget=Background
   properties {colorAndOpacity=<rgba 0.0137 0.0015 0.0513 0.2>}
   slot {class=GridSlot, properties {padding=<ltrb -16 -16 -16 -16>}}
  }
  Description {
   widgetClassPath=/MatchmakingSettingsFN1Shooters/MatchmakingSettings/WBP_FN1_Matchmaking_EntryDescription
   bIsVariable=true
   parent=Root
   widget=Description
   properties {visibility=Hidden}
   slot {class=GridSlot, properties {horizontalAlignment=HAlign_Center, row=1, padding=<ltrb 0 24 0 0>}}
  }
 ]
}
