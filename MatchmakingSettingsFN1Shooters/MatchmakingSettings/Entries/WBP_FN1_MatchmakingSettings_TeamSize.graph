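# Matchmaking settings row widget for choosing team size in FN1 Shooters, shown as a horizontal list of icon toggle buttons.
# It receives the team-size view model as its list item, and PopulateOptions wraps each option in an entry object with icon URL and chaos-mode flag.
# Selection writes the value back to the view model, and clicks fire analytics.
# The option description follows hover, or focus and selection on touch input.
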WidgetBlueprint /MatchmakingSettingsFN1Shooters/MatchmakingSettings/Entries/WBP_FN1_MatchmakingSettings_TeamSize
graphs [
 EventGraph {nodes [
   n1 = "On Item Selection Changed (ListView)"!() -> (OutputDelegate, then, Item:Object Reference, bIsSelected:bool) {is=title K2Node_ComponentBoundEvent}
   > Branch(Condition=@n1.bIsSelected)
   > n2 = Cast To VM_TeamSizeEnumEntry!(execute, Object=@n1.Item) -> (then, CastFailed, AsVM Team Size Enum Entry:VM Team Size Enum Entry Object Reference) {is=title K2Node_DynamicCast}
   > SetValue!(
    execute
    self:Fort Matchmaking Setting Discrete Enum VM Object Reference=Get FortMatchmakingSettingFN1ShooterTeamSizeVM() -> (FortMatchmakingSettingFN1ShooterTeamSizeVM:Fort Matchmaking Setting FN1Shooter Team Size VM Object Reference)
    Value:String=Get!(self:VM Team Size Enum Entry Object Reference=@n2) -> (Data_bIsEnabled:bool=~, Data_Value:String, Data_Name:Text).Data_Value
    LocalPlayer:Local Player Object Reference=GetOwningLocalPlayer!(self:Widget Object Reference) -> (ReturnValue:Local Player Object Reference)
   ) -> (then)
   > Branch(Condition=IsTouch!() -> (ReturnValue:bool))
   > UpdateDescriptionText!(
    execute
    self
    InSelectedButton:Object Reference=Get Selected Item!(self:List View Object Reference=Get ListView() -> (ListView:Common List View Object Reference)) -> (ReturnValue:Object Reference) {is=title K2Node_CallFunction}
   ) -> (then)
   n3 = "On Item Clicked (ListView)"!() -> (OutputDelegate, then, Item:Object Reference) {is=title K2Node_ComponentBoundEvent}
   > n4 = Cast To VM_TeamSizeEnumEntry(Object=@n3.Item)
   > n5 = Set!(
    execute
    AnalyticsAttribute_Name:String=GetKey!(self:Fort Matchmaking Setting Discrete VM Object Reference=Get FortMatchmakingSettingFN1ShooterTeamSizeVM()) -> (ReturnValue:String)
    AnalyticsAttribute_Value:String=Get(self=@n4) -> (Data_bIsEnabled=~).Data_Value
   ) -> (then, Output_Get_Name:String, Output_Get_Value:String)
   > Fire Event Element Interacted(
    Element=Self-Reference()
    AdditionalInfo="Add String (Non-Localized)"!(
     Info:"Flow Analytics Additional Info Structure (by ref)"=Make Flow Analytics Additional Info!() -> (FlowAnalyticsAdditionalInfo:Flow Analytics Additional Info Structure) {is=title K2Node_MakeStruct}
     Key:String=@n5
     Value:String=@n5.Output_Get_Value
    ) -> (ReturnValue:Flow Analytics Additional Info Structure)
   )
   n6 = "On Item Is Hovered Changed (ListView)"!() -> (OutputDelegate, then, Item:Object Reference, bIsHovered:bool) {is=title K2Node_ComponentBoundEvent}
   > SetVisibility(
    self=Get Description() -> (Description:WBP FN1 Matchmaking Entry Description Object Reference)
    InVisibility=Select(Option 0=Hidden, Option 1=SelfHitTestInvisible, Index=@n6.bIsHovered)
   )
   > Branch(Condition=@n6.bIsHovered)
   > UpdateDescriptionText(InSelectedButton=@n6.Item)
   Event OnAddedToFocusPath!() -> (OutputDelegate, then, InFocusEvent:Focus Event Structure) {is=title K2Node_Event}
   > ShowOutline!(
    execute
    self:WBP Matchmaking Settings Outline Object Reference=Get FocusOutline() -> (FocusOutline:WBP Matchmaking Settings Outline Object Reference)
    AdditionalCondition:bool=true
   ) -> (then)
   n7 = IsTouch()
   > Branch(Condition=@n7)
   n8 = Get Description()
   > SetVisibility(self=@n8, InVisibility=SelfHitTestInvisible)
   n9 = Event OnListItemObjectSet()
   > n10 = Cast To FortMatchmakingSettingFN1ShooterTeamSizeVM!(execute, Object=@n9.ListItemObject) -> (then, CastFailed, AsFort Matchmaking Setting FN1Shooter Team Size VM:Fort Matchmaking Setting FN1Shooter Team Size VM Object Reference) {is=title K2Node_DynamicCast}
   > Set FortMatchmakingSettingFN1ShooterTeamSizeVM(FortMatchmakingSettingFN1ShooterTeamSizeVM:Fort Matchmaking Setting FN1Shooter Team Size VM Object Reference=@n10) -> (Output_Get:Fort Matchmaking Setting FN1Shooter Team Size VM Object Reference)
   Event OnRemovedFromFocusPath!() -> (OutputDelegate, then, InFocusEvent:Focus Event Structure) {is=title K2Node_Event}
   > HideOutline(self:WBP Matchmaking Settings Outline Object Reference=Get FocusOutline())
   > Branch(Condition=@n7)
   > SetVisibility(self=@n8, InVisibility=Hidden)
  ]}
 FindSelectedItem {nodes [
   n1 = FindSelectedItem!() -> (then, OptionValue:String) {is=title K2Node_FunctionEntry}
   > n2 = For Each Loop with Break!(
    Exec:Exec
    Array:Array of Matchmaking Setting Option Structures=GetOptions!(self:Fort Matchmaking Setting Discrete Enum VM Object Reference=Get FortMatchmakingSettingFN1ShooterTeamSizeVM()) -> (ReturnValue:Array of Matchmaking Setting Option Structures)
    Break:Exec=@n4.then
   ) -> (LoopBody:Exec, Array Element:Matchmaking Setting Option Structure, Array Index:int, Completed:Exec)
   n3 = Break Matchmaking Setting Option!(MatchmakingSettingOption:"Matchmaking Setting Option Structure (by ref)"=@n2) -> (bIsEnabled:bool=~, Value:String, Name:Text) {is=title K2Node_BreakStruct}
   > Branch(Condition="Equal Exactly (String)"(A=@n1, B=@n3.Value))
   > "SetText (Text)"(self=Get Text!() -> (Text:Common Text Object Reference), InText=@n3.Name)
   > n4 = SetSelectedIndex!(execute, self:List View Object Reference=Get ListView(), Index:int=@n2.Array Index) -> (then)
  ]}
 GetAnalyticsAttribute {nodes [
   GetAnalyticsAttribute() {is=title K2Node_FunctionEntry}
   > Return Node!(execute, Attribute:Analytics Event Attr Structure=Get AnalyticsAttribute() -> (AnalyticsAttribute:Analytics Event Attr Structure))
  ]}
 IsTouch {nodes [
   IsTouch() {is=title K2Node_FunctionEntry}
   n1 = Get CommonInputSubsystem()
   > Return Node!(execute, ReturnValue:bool="Equal (Enum)"(A=GetCurrentInputType(self=@n1), B=Touch))
  ]}
 PopulateOptions {nodes [
   n1 = PopulateOptions!() -> (then, Options:Array of Matchmaking Setting Option Structures) {is=title K2Node_FunctionEntry}
   > Clear(TargetArray:Array of VM Team Size Enum Entry Object References=Get Entries() -> (Entries:Array of VM Team Size Enum Entry Object References))
   > n2 = For Each Loop(Array:Array of Matchmaking Setting Option Structures=@n1) -> (Array Element:Matchmaking Setting Option Structure)
   n3 = Break Matchmaking Setting Option(MatchmakingSettingOption=@n2) -> (bIsEnabled=~)
   n4 = Get FortMatchmakingSettingFN1ShooterTeamSizeVM()
   > n5 = Construct Object from Class!(
    execute
    Class=/MatchmakingSettingsFN1Shooters/MatchmakingSettings/Data/VM_TeamSizeEnumEntry.VM_TeamSizeEnumEntry_C
    self
    Data:Matchmaking Setting Option Structure=@n2
    IsPureChaos:bool=IsChaosMode!(self:Fort Matchmaking Setting FN1Shooter Team Size VM Object Reference=@n4, Value:String=@n3.Value) -> (ReturnValue:bool)
    ImageURL:String=GetTeamSizeIconURL!(self:Fort Matchmaking Setting FN1Shooter Team Size VM Object Reference=@n4, Value:String=@n3.Value) -> (ReturnValue:String)
   ) -> (then, ReturnValue:VM Team Size Enum Entry Object Reference)
   > Add(
    TargetArray:Array of VM Team Size Enum Entry Object References=Get Entries()
    NewItem:"VM Team Size Enum Entry Object Reference (by ref)"=@n5
   ) {is=title K2Node_CallArrayFunction}
   Set List Items!(execute=@n2.Completed, self:List View Object Reference=Get ListView(), InListItems:Array of Object References=Get Entries()) -> (then) {is=title K2Node_CallFunction}
   n6 = Get FortMatchmakingSettingFN1ShooterTeamSizeVM()
   n7 = GetValue!(self:Fort Matchmaking Setting Discrete Enum VM Object Reference=@n6) -> (ReturnValue:String)
   > FindSelectedItem!(execute, self, OptionValue:String=@n7) -> (then) {is=title K2Node_CallFunction}
   > Set!(execute, AnalyticsAttribute_Name=GetKey(self=@n6), AnalyticsAttribute_Value=@n7) -> (then, Output_Get:Analytics Event Attr Structure)
  ]}
 SetSettingType {nodes [
   n1 = SetSettingType!() -> (then, InText:Text) {is=title K2Node_FunctionEntry}
   > Set Text(Text=@n1, self=Get SecondText() -> (SecondText:Common Text Object Reference))
  ]}
 UpdateDescriptionText {nodes [
   n1 = UpdateDescriptionText!() -> (then, InSelectedButton:Object Reference) {is=title K2Node_FunctionEntry}
   > n2 = Cast To VM_TeamSizeEnumEntry(Object=@n1)
   > SetText(
    self:WBP FN1 Matchmaking Entry Description Object Reference=Get Description()
    InText:Text=GetOptionDescription!(
     self:Fort Matchmaking Setting Discrete Enum VM Object Reference=Get FortMatchmakingSettingFN1ShooterTeamSizeVM()
     Value:String=Get(self=@n2) -> (Data_bIsEnabled=~).Data_Value
    ) -> (ReturnValue:Text)
   )
  ]}
]
variables [
 AnalyticsAttribute {type=AnalyticsEventAttr, default {name="", value=""}}
 Entries {type="array<VM_TeamSizeEnumEntry_C>", default []}
]
widgets {
 info {inheritedWidgetCount=0, parentClass=&CommonUserWidget, rootWidgetClass=&GridPanel, widgetCount=9}
 widgets [
  Root {widgetClassPath=&GridPanel, bIsVariable=true, widget=Root, properties {columnFill [1], rowFill [1]}}
  Button_Hitbox {
   widgetClassPath=&Button
   parent=Root
   widget=Button_Hitbox
   properties {backgroundColor=<rgba 1 1 1 0>}
   slot {class=GridSlot, properties {columnSpan=2}}
  }
  VerticalBox_0 {
   widgetClassPath=&VerticalBox
   parent=Root
   widget=VerticalBox_0
   slot {class=GridSlot, properties {verticalAlignment=VAlign_Center, padding=<ltrb 0 0 32 0>}}
  }
  SecondText {
   widgetClassPath=&CommonTextBlock
   bIsVariable=true
   parent=VerticalBox_0
   widget=SecondText
   properties {
    autoWrapText=true
    bAutoCollapseWithEmptyText=true
    mobileFontSizeMultiplier=1.2
    scrollStyle=&TextScrollStyle-Base_C
    text=Setting Type
    colorAndOpacity {colorUseRule=UseColor_Specified, specifiedColor=<rgba 0.6939 0.6939 0.6939 0.749>}
    font {
     bForceMonospaced=false
     bMaterialIsStencil=false
     fontMaterial=null
     fontObject=&HeadingNow
     letterSpacing=0
     monospacedWidth=1
     size=15
     skewAmount=0
     typefaceFontName=75 Medium
     outlineSettings {
      bApplyOutlineToDropShadows=false
      bMiteredCorners=false
      bSeparateFillAlpha=false
      outlineMaterial=null
      outlineSize=0
      outlineColor=<rgba 0 0 0 1>
     }
    }
    shadowOffset=<0 0>
   }
   slot {class=VerticalBoxSlot}
  }
  Text {
   widgetClassPath=&CommonTextBlock
   bIsVariable=true
   parent=VerticalBox_0
   widget=Text
   properties {
    bAutoCollapseWithEmptyText=true
    mobileFontSizeMultiplier=1.2
    scrollStyle=&TextScrollStyle-Base_C
    text=$SettingName
    textTransformPolicy=ToUpper
    font {
     bForceMonospaced=false
     bMaterialIsStencil=false
     fontMaterial=null
     fontObject=&HeadingNow
     letterSpacing=0
     monospacedWidth=1
     size=15
     skewAmount=0
     typefaceFontName=86 Bold
     outlineSettings {
      bApplyOutlineToDropShadows=false
      bMiteredCorners=false
      bSeparateFillAlpha=false
      outlineMaterial=null
      outlineSize=0
      outlineColor=<rgba 0 0 0 1>
     }
    }
    shadowOffset=<0 0>
   }
   slot {class=VerticalBoxSlot, properties {verticalAlignment=VAlign_Center}}
  }
  FocusOutline {
   widgetClassPath=/MatchmakingSettingsFN1Shooters/MatchmakingSettings/WBP_Matchmaking_Settings_Outline
   bIsVariable=true
   parent=Root
   widget=FocusOutline
   slot {class=GridSlot, properties {columnSpan=2, layer=1, padding=<ltrb -28 -24 -28 -24>}}
  }
  ListView {
   widgetClassPath=&CommonListView
   bIsVariable=true
   parent=Root
   widget=ListView
   properties {
    bAllowDragging=false
    bEnableRightClickScrolling=false
    bReturnFocusToSelection=true
    bSelectItemOnNavigation=false
    clipping=Inherit
    consumeMouseWheel=Never
    entryWidgetClass=/MatchmakingSettingsFN1Shooters/MatchmakingSettings/Entries/Buttons/WBP_FN1_ToggleIconButton.WBP_FN1_ToggleIconButton_C
    horizontalEntrySpacing=18
    numDesignerPreviewEntries=3
    orientation=Orient_Horizontal
    scrollBarStyle {
     thickness=0
     draggedThumbImage {
      bIsDynamicallyLoaded=false
      drawAs=NoDrawType
      imageType=FullColor
      mirroring=NoMirror
      resourceName=../../../Engine/Content/Slate/Common/Scrollbar_Thumb.png
      resourceObject=null
      tiling=NoTile
      imageSize=<0 0>
      margin=<ltrb 0.25 0.25 0.25 0.25>
      outlineSettings {
       bUseBrushTransparency=false
       roundingType=HalfHeightRadius
       width=0
       color {colorUseRule=UseColor_Specified, specifiedColor=<rgba 0 0 0 0>}
       cornerRadii=<0 0 0 0>
      }
      tintColor {colorUseRule=UseColor_Specified, specifiedColor=<rgba 1 1 1 1>}
      uVRegion {bIsValid=false, max=<0 0>, min=<0 0>}
     }
     horizontalBackgroundImage {
      bIsDynamicallyLoaded=false
      drawAs=NoDrawType
      imageType=NoImage
      mirroring=NoMirror
      resourceName=None
      resourceObject=null
      tiling=NoTile
      imageSize=<0 0>
      margin=<ltrb 0 0 0 0>
      outlineSettings {
       bUseBrushTransparency=false
       roundingType=HalfHeightRadius
       width=0
       color {colorUseRule=UseColor_Specified, specifiedColor=<rgba 0 0 0 0>}
       cornerRadii=<0 0 0 0>
      }
      tintColor {colorUseRule=UseColor_Specified, specifiedColor=<rgba 1 1 1 1>}
      uVRegion {bIsValid=false, max=<0 0>, min=<0 0>}
     }
     horizontalBottomSlotImage {
      bIsDynamicallyLoaded=false
      drawAs=NoDrawType
      imageType=FullColor
      mirroring=NoMirror
      resourceName=../../../Engine/Content/Slate/Common/Scrollbar_Background_Horizontal.png
      resourceObject=null
      tiling=NoTile
      imageSize=<8 8>
      margin=<ltrb 0 0 0 0>
      outlineSettings {
       bUseBrushTransparency=false
       roundingType=HalfHeightRadius
       width=0
       color {colorUseRule=UseColor_Specified, specifiedColor=<rgba 0 0 0 0>}
       cornerRadii=<0 0 0 0>
      }
      tintColor {colorUseRule=UseColor_Specified, specifiedColor=<rgba 1 1 1 1>}
      uVRegion {bIsValid=false, max=<0 0>, min=<0 0>}
     }
     horizontalTopSlotImage {
      bIsDynamicallyLoaded=false
      drawAs=NoDrawType
      imageType=FullColor
      mirroring=NoMirror
      resourceName=../../../Engine/Content/Slate/Common/Scrollbar_Background_Horizontal.png
      resourceObject=null
      tiling=NoTile
      imageSize=<8 8>
      margin=<ltrb 0 0 0 0>
      outlineSettings {
       bUseBrushTransparency=false
       roundingType=HalfHeightRadius
       width=0
       color {colorUseRule=UseColor_Specified, specifiedColor=<rgba 0 0 0 0>}
       cornerRadii=<0 0 0 0>
      }
      tintColor {colorUseRule=UseColor_Specified, specifiedColor=<rgba 1 1 1 1>}
      uVRegion {bIsValid=false, max=<0 0>, min=<0 0>}
     }
     hoveredThumbImage {
      bIsDynamicallyLoaded=false
      drawAs=NoDrawType
      imageType=FullColor
      mirroring=NoMirror
      resourceName=../../../Engine/Content/Slate/Common/Scrollbar_Thumb.png
      resourceObject=null
      tiling=NoTile
      imageSize=<0 0>
      margin=<ltrb 0.25 0.25 0.25 0.25>
      outlineSettings {
       bUseBrushTransparency=false
       roundingType=HalfHeightRadius
       width=0
       color {colorUseRule=UseColor_Specified, specifiedColor=<rgba 0 0 0 0>}
       cornerRadii=<0 0 0 0>
      }
      tintColor {colorUseRule=UseColor_Specified, specifiedColor=<rgba 1 1 1 1>}
      uVRegion {bIsValid=false, max=<0 0>, min=<0 0>}
     }
     normalThumbImage {
      bIsDynamicallyLoaded=false
      drawAs=NoDrawType
      imageType=FullColor
      mirroring=NoMirror
      resourceName=../../../Engine/Content/Slate/Common/Scrollbar_Thumb.png
      resourceObject=null
      tiling=NoTile
      imageSize=<0 0>
      margin=<ltrb 0.25 0.25 0.25 0.25>
      outlineSettings {
       bUseBrushTransparency=false
       roundingType=HalfHeightRadius
       width=0
       color {colorUseRule=UseColor_Specified, specifiedColor=<rgba 0 0 0 0>}
       cornerRadii=<0 0 0 0>
      }
      tintColor {colorUseRule=UseColor_Specified, specifiedColor=<rgba 1 1 1 1>}
      uVRegion {bIsValid=false, max=<0 0>, min=<0 0>}
     }
     verticalBackgroundImage {
      bIsDynamicallyLoaded=false
      drawAs=NoDrawType
      imageType=NoImage
      mirroring=NoMirror
      resourceName=None
      resourceObject=null
      tiling=NoTile
      imageSize=<0 0>
      margin=<ltrb 0 0 0 0>
      outlineSettings {
       bUseBrushTransparency=false
       roundingType=HalfHeightRadius
       width=0
       color {colorUseRule=UseColor_Specified, specifiedColor=<rgba 0 0 0 0>}
       cornerRadii=<0 0 0 0>
      }
      tintColor {colorUseRule=UseColor_Specified, specifiedColor=<rgba 1 1 1 1>}
      uVRegion {bIsValid=false, max=<0 0>, min=<0 0>}
     }
     verticalBottomSlotImage {
      bIsDynamicallyLoaded=false
      drawAs=NoDrawType
      imageType=FullColor
      mirroring=NoMirror
      resourceName=../../../Engine/Content/Slate/Common/Scrollbar_Background_Vertical.png
      resourceObject=null
      tiling=NoTile
      imageSize=<8 8>
      margin=<ltrb 0 0 0 0>
      outlineSettings {
       bUseBrushTransparency=false
       roundingType=HalfHeightRadius
       width=0
       color {colorUseRule=UseColor_Specified, specifiedColor=<rgba 0 0 0 0>}
       cornerRadii=<0 0 0 0>
      }
      tintColor {colorUseRule=UseColor_Specified, specifiedColor=<rgba 1 1 1 1>}
      uVRegion {bIsValid=false, max=<0 0>, min=<0 0>}
     }
     verticalTopSlotImage {
      bIsDynamicallyLoaded=false
      drawAs=NoDrawType
      imageType=FullColor
      mirroring=NoMirror
      resourceName=../../../Engine/Content/Slate/Common/Scrollbar_Background_Vertical.png
      resourceObject=null
      tiling=NoTile
      imageSize=<8 8>
      margin=<ltrb 0 0 0 0>
      outlineSettings {
       bUseBrushTransparency=false
       roundingType=HalfHeightRadius
       width=0
       color {colorUseRule=UseColor_Specified, specifiedColor=<rgba 0 0 0 0>}
       cornerRadii=<0 0 0 0>
      }
      tintColor {colorUseRule=UseColor_Specified, specifiedColor=<rgba 1 1 1 1>}
      uVRegion {bIsValid=false, max=<0 0>, min=<0 0>}
     }
    }
   }
   slot {class=GridSlot, properties {column=1, horizontalAlignment=HAlign_Right, verticalAlignment=VAlign_Center}}
  }
  Background {
   widgetClassPath=&Image
   bIsVariable=true
   parent=Root
   widget=Background
   properties {colorAndOpacity=<rgba 0.0137 0.0015 0.0513 0.2>}
   slot {class=GridSlot, properties {columnSpan=2, padding=<ltrb -16 -16 -16 -16>}}
  }
  Description {
   widgetClassPath=/MatchmakingSettingsFN1Shooters/MatchmakingSettings/WBP_FN1_Matchmaking_EntryDescription
   bIsVariable=true
   parent=Root
   widget=Description
   properties {visibility=Hidden}
   slot {class=GridSlot, properties {columnSpan=2, horizontalAlignment=HAlign_Center, row=1, padding=<ltrb 0 24 0 0>}}
  }
 ]
}
